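; Parent Blueprint for the Athena floor launch pad trap, handling launch feedback and launching vehicles.
; On placement it registers the collision trigger and, on physically simulated islands, configures a physics impulse component.
; Vehicle overlaps keep horizontal velocity, add upward impulse and pitch by travel direction.
; Trap events and GameplayCues drive sounds, particles and team-filtered sound indicators; player launch logic is absent here, only its effects.
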(event EventOnPlaced
  ;[1 1 1] Initial Setup
  (Trap|AddTriggerComponent (Variables|Default|GetS_Athena_Launchpad_Collision))
  ;; FortPhysicsImpulse Setup
  (if (Fortnite|IsPhysicallySimulatedIsland)
    (Class|FortPhysicsImpulseComponent|SetImpulseDeviceInputLaunchPad (Variables|Default|GetFortPhysicsImpulse) "LaunchPad" (Utilities|Struct|MakeFortPhysicsLaunchPadDeviceInput (Variables|Default|GetPhysicsPropVelocityOrImpulse) (Math|Conversions|ToInteger(Byte) (Utilities|Enum|LiteralenumEAllowedToBounce "Objects")) (Variables|Default|GetImpulseOrVelocity) true)))
  (Audio|PlaySoundatLocation (Variables|Audio|GetTrap_Placed_Sound) (Transformation|GetWorldLocation (Variables|Default|GetPlacementSoundLocation)))
  (Utilities|TrapPlacementDustFX :Attach%20to%20Component (Variables|Debug|GetStaticMeshComponent) :Socket "PlacedFX" :Rotation "0, 0, 0" :Burst%20Rate 8 :Size%20%28Volume%29 "100.000000,100.000000,15.000000" :Velocity%20Scale "1,1,1" :Size%20Scale%20%28Particles%29 1))

(event EventOnFinishedBuilding
  ;[1 1 1] Finished building - This is never called in the Thrown version of the item, because it isn't being built by a player.
  (Audio|PlaySoundatLocation (Variables|Audio|GetTrap_Active_Sound) (Transformation|GetActorLocation))
  (Fortnite|ShowSoundIndicatorLocalforTeam :Location (Transformation|GetActorLocation) :MaxAudibleDistance (Variables|/Script/Engine.SoundBase|GetMaxDistance (Variables|Audio|GetTrap_Placed_Sound)) :IndicatorType "Gunshot" :Team (Team|GetActorTeam self) :Affiliations (Utilities|Array|MakeArray "Hostile" "Neutral") :Tint "(R=0.000000,G=0.000000,B=0.000000,A=0.000000)" :OverrideIcon "/Game/UI/Foundation/Textures/Icons/Navigation/t_icon_nav_mobile_audioviz_building_32.t_icon_nav_mobile_audioviz_building_32" :AttenuationMaxRangeMultiplier 1.0 :AttenuationIntensityMultiplier 1.0))

(event GameplayCue|GameplayCue.Abilities.Activation.Traps.ActivateTrap
  (params (EventType :type EGameplayCueEvent%20Enum :default "OnActive") (Parameters :type Gameplay%20Cue%20Parameters%20Structure%20%28by%20ref%29))
  ;[1 1 1] On Fire
  (switch Utilities|FlowControl|Switch|SwitchonEGameplayCueEvent EventType
    (:Executed
      (Audio|PlaySoundatLocation (Variables|Audio|GetTrap_Fire_Sound) (Transformation|GetActorLocation) "0, 0, 0" 1.5 0.75))))

(event GameplayCue|GameplayCue.Abilities.Activation.Traps.ReloadBegin
  (params (EventType :type EGameplayCueEvent%20Enum :default "OnActive") (Parameters :type Gameplay%20Cue%20Parameters%20Structure%20%28by%20ref%29))
  ;[1 1 1] Reload Begin
  (switch Utilities|FlowControl|Switch|SwitchonEGameplayCueEvent EventType
    (:Executed
      (Audio|PlaySoundatLocation (Variables|Audio|GetTrap_Reload_Sound) (Transformation|GetActorLocation)))))

(event EventOnReloadEnd
  ;[1 1 1] Reload End
  (Audio|PlaySoundatLocation (Variables|Audio|GetTrap_Active_Sound) (Transformation|GetActorLocation)))

(event EventOnOutOfDurability
  ;[1 1 1] Durability Change
  (Utilities|FlowControl|Delay 1)
  (Audio|PlaySoundatLocation (Variables|Audio|GetTrap_Explode_Sound) (Transformation|GetActorLocation))
  (Effects|SpawnEmitteratLocation "/Game/Weapons/FORT_TRAPS/Spike_Trap/P_Spike_Trap_Destruction.P_Spike_Trap_Destruction" (Transformation|GetWorldLocation (Variables|Transformation|GetRootComponent))))

(event EventTriggerLaunchEffects
  (params (Pawn :type Fort%20Pawn%20Object%20Reference))
  ;[1 1 1] EVENT -> Trigger Launch FX
  (Components|Activation|SetActive (Variables|Default|GetP_Athena_LaunchPad_temp) true true)
  (Fortnite|ShowSoundIndicatorLocalforTeam :Location (Transformation|GetActorLocation) :MaxAudibleDistance (Variables|/Script/Engine.SoundBase|GetMaxDistance (Variables|Audio|GetTrap_Launch_3P_Sound)) :IndicatorType "Glider" :Team (Class|FortPawn|GetTeam Pawn) :Affiliations (Utilities|Array|MakeArray "Hostile") :Tint "(R=0.000000,G=0.000000,B=0.000000,A=0.000000)" :OverrideIcon "/Game/UI/Foundation/Textures/Icons/Navigation/SV_Launcher.SV_Launcher" :AttenuationMaxRangeMultiplier 1.0 :AttenuationIntensityMultiplier 1.0))

(event EventOnDeathServer
  (params (Damage :type Float%20%28single-precision%29) (DamageTags :type Gameplay%20Tag%20Container%20Structure%20%28by%20ref%29) (Momentum :type Vector :default "0, 0, 0") (HitInfo :type Hit%20Result%20Structure%20%28by%20ref%29) (InstigatedBy :type Controller%20Object%20Reference) (DamageCauser :type Actor%20Object%20Reference) (EffectContext :type Gameplay%20Effect%20Context%20Handle%20Structure))
  ;[1 1 1] EVENT -> On Death Server
  (Rendering|SetActorHiddenInGame true))

(event OnComponentBeginOverlap(S_Athena_Launchpad_Collision)
  (params (OverlappedComponent :type Primitive%20Component%20Object%20Reference) (OtherActor :type Actor%20Object%20Reference) (OtherComp :type Primitive%20Component%20Object%20Reference) (OtherBodyIndex :type Integer) (bFromSweep :type Boolean) (SweepResult :type Hit%20Result%20Structure%20%28by%20ref%29))
  ;[1 1 1] EVENT -> OnOverlap: Launch a vehicle or physics prop on overlap
  (bind _asfort_athena_vehicle (Utilities|Casting|CastToFortAthenaVehicle OtherActor))
  (bind (_returnvalue_x _returnvalue_y) (Math|Vector|BreakVector (Vehicle|GetVehicleVelocity _asfort_athena_vehicle)))
  (Audio|BroadcastAbilityImpactatLocation 0 _asfort_athena_vehicle (Math|Vector|MakeVector _returnvalue_x _returnvalue_y (ScalableFloat|GetValueatLevel (Variables|Default|GetVehicleZ_LaunchMagnitude))) 777.0 "None" false) ; Launch vehicle up and retain forward speed
  (Audio|BroadcastAbilitySpinThrustatLocation 0 _asfort_athena_vehicle (CallFunction|CalculateVehiclePitchDirection _asfort_athena_vehicle (ScalableFloat|GetValueatLevel (Variables|Default|GetVehiclePitchMagnitude))) 777.0) ; Pitch vehicle forwards or backwards for a more dynamic feel
  (Components|Activation|SetActive (Variables|Default|GetP_Athena_LaunchPad_temp) true true) ; Activate Launch FX
  (Audio|PlaySoundatLocation (Variables|Default|GetVehicleActivate_3P_Sound) (Transformation|GetActorLocation)) ; Activation sound when a vehicle hits the launch pad
)

(event OnPhysicsImpulseOverlapTriggeredSignatureDelegate(FortPhysicsImpulse)
  (params (Payload :type Fort%20Physics%20Impulse%20Component%20Delegate%20Payload%20Structure))
  ;[1 1 1] Newton: Activate Launch FX
  (if (Fortnite|IsPhysicallySimulatedIsland)
    (Components|Activation|SetActive (Variables|Default|GetP_Athena_LaunchPad_temp) true true)
    (Audio|PlaySoundatLocation (Variables|Audio|GetTrap_Launch_3P_Sound) (Transformation|GetActorLocation))))

(fn CalculateVehiclePitchDirection
  (params (Vehicle :type Fort%20Athena%20Vehicle%20Object%20Reference) (PitchMagnitude :type Float%20%28double-precision%29))
  (return (* (Math|Vector|SelectVector (Math|Vector|MirrorVectorbyNormal (Vehicle|GetVehicleRightVector Vehicle) (Vehicle|GetVehicleRightVector Vehicle)) (Vehicle|GetVehicleRightVector Vehicle) (> (Vehicle|GetForwardSpeedKmh Vehicle) 0.0)) PitchMagnitude)))

(fn ConstructionScript)
