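; Blueprint for the Athena player launch pad floor trap, a buildable placed by players.
; When building finishes, it runs the parent logic, then on non-server instances casts the owning player state's owner to the Athena player controller and registers a sound indicator placement context for the trap.
; Small overrides let linear movers target it, disallow invisible props, and pass the marker details ID through from the parent.
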(event EventOnFinishedBuilding
  (|CallParentFunction)
  ;[1 1 1] Ping Squad Mates after construction
  (if (not (Networking|IsServer))
    (bind _asfort_player_controller_athena (Utilities|Casting|CastToFortPlayerControllerAthena (Actor|GetOwner (Building|GetOwningPlayerState))))
    (Fortnite|RegisterSoundIndicatorPlacementContext _asfort_player_controller_athena self)))

(fn GetMarkerDetailsId
  (return (|CallParentFunction)))

(fn IsTargetableByLinearMover
  (return true))

(fn ShouldInvisiblePropsBeAllowed
  (return false))

(fn ConstructionScript)
